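# Sound cue for the editor's Game Preview sounds, named for ejecting from the player during play preview.
# A single wave player feeds the EjectFromPlayer sound wave straight to the output at reduced volume.
# It overrides concurrency to prevent new instances from starting once the limit is reached, and sets a very large max distance.
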SoundCue /Engine/EditorSounds/GamePreview/EjectFromPlayer_Cue
root=WavePlayer_102
nodes [{id=WavePlayer_102, properties {soundWaveAssetPtr=&/EjectFromPlayer}}]
properties {bOverrideConcurrency=true, maxDistance=2097152, volumeMultiplier=0.2, concurrencyOverrides {resolutionRule=PreventNew}}
