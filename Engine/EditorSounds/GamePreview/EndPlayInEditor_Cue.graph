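# EndPlayInEditor_Cue is an engine editor Sound Cue, named for the end of a Play In Editor session.
# A single wave player feeds the EndPlayInEditor sound wave straight to the output at reduced volume.
# It overrides concurrency with a prevent-new rule, so extra instances are rejected once the limit is reached.
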SoundCue /Engine/EditorSounds/GamePreview/EndPlayInEditor_Cue
root=WavePlayer_103
nodes [{id=WavePlayer_103, properties {soundWaveAssetPtr=&/EndPlayInEditor}}]
properties {bOverrideConcurrency=true, maxDistance=2097152, volumeMultiplier=0.15, concurrencyOverrides {resolutionRule=PreventNew}}
