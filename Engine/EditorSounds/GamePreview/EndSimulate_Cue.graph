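# EndSimulate_Cue is an editor sound cue in the GamePreview set that, going by its name, plays when a Simulate session ends.
# A single wave player feeds the EndSimulate wave straight to the output at reduced volume.
# It overrides concurrency with a prevent-new rule, so extra instances are refused once the limit is reached.
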SoundCue /Engine/EditorSounds/GamePreview/EndSimulate_Cue
root=WavePlayer_104
nodes [{id=WavePlayer_104, properties {soundWaveAssetPtr=&/EndSimulate}}]
properties {bOverrideConcurrency=true, maxDistance=2097152, volumeMultiplier=0.25, concurrencyOverrides {resolutionRule=PreventNew}}
