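# PossessPlayer_Cue is an editor SoundCue in the GamePreview sounds, played for possessing the player, judging by its name.
# A single wave player node feeds the PossessPlayer sound wave straight to the output at reduced volume.
# It overrides concurrency with a prevent-new rule, so extra instances do not start once the limit is reached, and sets a very large max distance.
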SoundCue /Engine/EditorSounds/GamePreview/PossessPlayer_Cue
root=WavePlayer_105
nodes [{id=WavePlayer_105, properties {soundWaveAssetPtr=&/PossessPlayer}}]
properties {bOverrideConcurrency=true, maxDistance=2097152, volumeMultiplier=0.25, concurrencyOverrides {resolutionRule=PreventNew}}
