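# StartPlayInEditor_Cue is an editor Sound Cue in the GamePreview folder, named for the start of a Play In Editor session.
# A single Wave Player node feeds the StartPlayInEditor wave straight to the output at reduced volume.
# The cue overrides concurrency to prevent new instances once its limit is reached, and sets a very large max distance.
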SoundCue /Engine/EditorSounds/GamePreview/StartPlayInEditor_Cue
root=WavePlayer_106
nodes [{id=WavePlayer_106, properties {soundWaveAssetPtr=&/StartPlayInEditor}}]
properties {bOverrideConcurrency=true, maxDistance=2097152, volumeMultiplier=0.25, concurrencyOverrides {resolutionRule=PreventNew}}
