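# StartSimulate_Cue is an editor Sound Cue in the GamePreview set, named for the start of a Simulate session.
# A single Wave Player node plays the StartSimulate wave directly as the root, at reduced volume.
# Its own concurrency override refuses new instances once the limit is reached, rather than stopping ones already playing.
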SoundCue /Engine/EditorSounds/GamePreview/StartSimulate_Cue
root=WavePlayer_107
nodes [{id=WavePlayer_107, properties {soundWaveAssetPtr=&/StartSimulate}}]
properties {bOverrideConcurrency=true, maxDistance=2097152, volumeMultiplier=0.25, concurrencyOverrides {resolutionRule=PreventNew}}
